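# Looping fuel-warning sound cue for the BikeKid guided missile, routed to the Fortnite weapons sound class.
# A mixer layers the Athena jetpack fuel-warning wave, modulated to double pitch, with a quieter Rocket_FuelWarning_04 wave, and a looping node repeats the mix.
# It references jetpack attenuation but overrides it with its own falloff and occlusion channel.
# Concurrency overrides allow one instance and block new plays while it runs.
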SoundCue /BikeKid/Sounds/Guided_Missile/Rocket_FuelWarning_Cue
root=Looping_0
nodes [
 {id=Looping_0, children [Mixer_0]}
 {id=Mixer_0, children [Modulator_0,WavePlayer_2], properties {inputVolume [1,0.68]}}
 {id=Modulator_0, children [WavePlayer_1], properties {pitchMax=2, pitchMin=2}}
 {id=WavePlayer_1, properties {soundWaveAssetPtr=/Game/Athena/Sounds/Items/Jetpack/Jetpack_FuelWarning_01}}
 {id=WavePlayer_2, properties {soundWaveAssetPtr=&/Rocket_FuelWarning_04}}
]
properties {
 attenuationSettings=/Game/Sounds/Attenuation/Fort_Jetpack_Att
 bOverrideAttenuation=true
 bOverrideConcurrency=true
 maxDistance=2000
 soundClassObject=/Game/Sounds/FortSoundClassesAndModes/Fort_Weapons
 volumeMultiplier=0.6
 attenuationOverrides {falloffDistance=1400, occlusionTraceChannel=ECC_GameTraceChannel8, attenuationShapeExtents {x=600}}
 concurrencyOverrides {maxCount=1, resolutionRule=PreventNew}
}
