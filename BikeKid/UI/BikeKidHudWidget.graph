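# BikeKid HUD widget, derived from native BikeKidHUD, showing an energy-driven health bar and low-health, weak-signal and operator-damage warnings.
# Energy changes set the bar material's Progress; the low threshold drives a color animation, warning visibility and a once-gated fuel-warning sound.
# Operator damage shows a warning that a looping timer later hides, while range functions toggle weak signal.
# Signal loss or dismissal collapses the main overlay; Construct applies mobile sizing.
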WidgetBlueprint /BikeKid/UI/BikeKidHudWidget
animations [Anim_LowHealth {
  binding_count=1
  length_seconds=0.0834
  object=&self:Anim_LowHealth
  sequence {
   bindings [Image_HealthBar {
     id=#1
     possessed_class=Image
     tracks [{
       type=MovieSceneWidgetMaterialTrack
       sections [{
         type=MovieSceneParameterSection
         end_frame=5
         start_frame=0
         channels [
          Progress_Color_StartR {type=MovieSceneScriptingFloatChannel, keys [[0,0.718],[5,1]]}
          Progress_Color_StartG {type=MovieSceneScriptingFloatChannel, keys [[0,0.733],[5,0]]}
          Progress_Color_StartB {type=MovieSceneScriptingFloatChannel, keys [[0,0.745],[5,0]]}
          Progress_Color_StartA {type=MovieSceneScriptingFloatChannel, const [1,1]}
          Progress_Color_StopR {type=MovieSceneScriptingFloatChannel, keys [[0,0.718],[5,1]]}
          Progress_Color_StopG {type=MovieSceneScriptingFloatChannel, keys [[0,0.733],[5,0.75]]}
          Progress_Color_StopB {type=MovieSceneScriptingFloatChannel, keys [[0,0.745],[5,0]]}
          Progress_Color_StopA {type=MovieSceneScriptingFloatChannel, const [1,1]}
          Marker_ColorR {type=MovieSceneScriptingFloatChannel, const [1,2]}
          Marker_ColorG {type=MovieSceneScriptingFloatChannel, keys [[0,1],[5,0]]}
          Marker_ColorB {type=MovieSceneScriptingFloatChannel, keys [[0,1],[5,0]]}
          Marker_ColorA {type=MovieSceneScriptingFloatChannel, const [1,1]}
         ]
        }]
      }]
    }]
   display_rate {numerator=60}
  }
 }]
graphs [
 EventGraph {nodes [
   n1 = Event OnBikeKidEnergyChanged!() -> (OutputDelegate, then, EnergyPercentage:float) {is=title K2Node_Event}
   > UpdateEnergyBar!(execute, self, HealthPercentage:float=@n1.EnergyPercentage) -> (then)
   n2 = Get Low Energy Threshold() -> (Low Energy Threshold:double)
   > n3 = Branch(Condition=AND Boolean(A="float <= float"(A=@n1.EnergyPercentage, B=@n2) -> (ReturnValue=~), B="float > float"(A=@n1.EnergyPercentage)))
   > Do Once(Reset=@n3.else)
   > n4 = SpawnSound2D!(
    execute
    Sound:Sound Base Object Reference=/BikeKid/Sounds/Guided_Missile/Rocket_FuelWarning_Cue
    VolumeMultiplier:float=1
    PitchMultiplier:float=1
    StartTime:float=0
    ConcurrencySettings:Sound Concurrency Object Reference
    bPersistAcrossLevelTransition:bool=false
    bAutoDestroy:bool=true
   ) -> (then, ReturnValue:Audio Component Object Reference)
   > Set LowEnergyAudioComponent(LowEnergyAudioComponent:Audio Component Object Reference=@n4) -> (Output_Get:Audio Component Object Reference)
   Event Destruct()
   n5 = Get LowEnergyAudioComponent() -> (LowEnergyAudioComponent:Audio Component Object Reference)
   > Is Valid(InputObject=@n5)
   > Stop(self:Audio Component Object Reference=@n5)
   Event OnOwningPlayerDamaged() {is=title K2Node_Event}
   > OnPlayerDamaged()
   Event OnBikeKidDismissAbilityUsed() {is=title K2Node_Event}
   > SetVisibility(self=Get Overlay_Main() -> (Overlay_Main:Overlay Object Reference), InVisibility=Collapsed)
   n6 = Get LowEnergyAudioComponent()
   > Is Valid(InputObject=@n6)
   > Stop(self=@n6)
   Event Construct()
   > SetMobileSize()
  ]}
 OnEnterWarningRange {nodes [
   OnEnterWarningRange() {is=title K2Node_FunctionEntry}
   > SetDistanceWarningVisuals!(execute, self, ShowWarning:bool=true) -> (then)
  ]}
 OnExitWarningRange {nodes [
   OnExitWarningRange() {is=title K2Node_FunctionEntry}
   > SetDistanceWarningVisuals(ShowWarning=false)
  ]}
 OnPlayerDamaged {nodes [
   OnPlayerDamaged() {is=title K2Node_FunctionEntry}
   > n1 = Set PlayerDamageTimer(PlayerDamageTimer:Timer Handle Structure) -> (Output_Get:Timer Handle Structure)
   > n2 = Branch(Condition=Is Timer Active by Handle!(Handle:Timer Handle Structure=@n1) -> (ReturnValue:bool))
   SetVisibility(
    execute=@n2.else
    self=Get Overlay_PlayerDamage_Warning() -> (Overlay_PlayerDamage_Warning:Overlay Object Reference)
    InVisibility=SelfHitTestInvisible
   )
   > n3 = Set Timer by Function Name!(
    execute
    Object:Object Reference
    FunctionName:String=UpdatePlayerDamageWarning
    Time:float=1
    bLooping:bool=true
    bMaxOncePerFrame:bool=false
    InitialStartDelay:float=0
    InitialStartDelayVariance:float=0
   ) -> (then, ReturnValue:Timer Handle Structure) {is=title K2Node_CallFunction}
   > Set PlayerDamageTimer(PlayerDamageTimer=@n3)
  ]}
 OnSignalLost {nodes [
   OnSignalLost() {is=title K2Node_FunctionEntry}
   > SetVisibility(self=Get Overlay_Main(), InVisibility=Collapsed)
  ]}
 SetDistanceWarningVisuals {nodes [
   n1 = SetDistanceWarningVisuals!() -> (then, ShowWarning:bool=~) {is=title K2Node_FunctionEntry}
   > SetVisibility(
    self=Get Overlay_WeakSignal_Warning() -> (Overlay_WeakSignal_Warning:Overlay Object Reference)
    InVisibility=Select(Option 0=Collapsed, Option 1=SelfHitTestInvisible, Index=@n1)
   )
  ]}
 SetMobileSize {nodes [
   SetMobileSize() {is=title K2Node_FunctionEntry}
   n1 = IsMobileGame()
   > SetRenderTranslation(
    self=Get HorizontalBox_HealthText() -> (HorizontalBox_HealthText:Horizontal Box Object Reference)
    Translation=Select(Option 0:Vector 2D Structure=(X=5,Y=-5), Option 1:Vector 2D Structure=(X=2,Y=-10), Index=@n1) -> (ReturnValue:Vector 2D Structure)
   )
   > SetRenderScale!(
    execute
    self:Widget Object Reference=Get Image_HealthIcon() -> (Image_HealthIcon:Image Object Reference)
    Scale:Vector 2D Structure=Select(Option 0=(X=1,Y=1), Option 1=(X=1.9,Y=1.9), Index=@n1)
   ) -> (then)
   > SetRenderScale(Scale=Select(Option 0=(X=1,Y=1), Option 1=(X=1.35,Y=1.35), Index=@n1))
   > SetRenderTranslation(
    self=Get HorizontalBox_DistanceText() -> (HorizontalBox_DistanceText:Horizontal Box Object Reference)
    Translation=Select(Option 0=(X=0,Y=-6), Option 1=(X=0,Y=-6), Index=@n1)
   )
   > SetRenderScale(self=Get HorizontalBox_DistanceText(), Scale=Select(Option 0=(X=1,Y=1), Option 1=(X=1.05,Y=1.05), Index=@n1))
   > SetRenderScale(Scale=Select(Option 0=(X=1,Y=1), Option 1=(X=1.2,Y=1.2), Index=@n1))
  ]}
 UpdateEnergyBar {nodes [
   n1 = UpdateEnergyBar!() -> (then, HealthPercentage:float=~) {is=title K2Node_FunctionEntry}
   > n2 = GetDynamicMaterial(self=Get Image_HealthBar() -> (Image_HealthBar:Image Object Reference))
   > SetScalarParameterValue(self=@n2, ParameterName=Progress, Value=@n1)
   > Sequence()
   n3 = Get Anim_LowHealth() -> (Anim_LowHealth:Widget Animation Object Reference)
   n4 = "float <= float"(A=Get HealthPercentage() -> (HealthPercentage:float), B=Get Low Energy Threshold()) -> (ReturnValue=~)
   > PlayAnimation(
    InAnimation=@n3
    StartAtTime=SelectFloat(A=0, B=GetEndTime(self=@n3), bPickA=@n4)
    PlayMode=Select(Option 0:EUMGSequencePlayMode Enum=Reverse, Option 1:EUMGSequencePlayMode Enum, Index=@n4) -> (ReturnValue:EUMGSequencePlayMode Enum)
   )
   Reroute Node()
   > Branch(Condition="float <= float"(A=@n1) -> (ReturnValue=~))
   > DoOnce MultiInput!(A In:Exec, Reset In:Exec) -> (A Out:Exec, Reset Out:Exec) {is=title K2Node_DoOnceMultiInput}
   n5 = Get LowEnergyAudioComponent()
   > Is Valid(InputObject=@n5)
   > Stop(self=@n5)
   SetVisibility(
    self=Get Overlay_LowHealth_Warning() -> (Overlay_LowHealth_Warning:Overlay Object Reference)
    InVisibility=Select(Option 0:ESlateVisibility Enum=Hidden, Option 1:ESlateVisibility Enum=SelfHitTestInvisible, Index=@n4) -> (ReturnValue:ESlateVisibility Enum)
   )
  ]}
 UpdatePlayerDamageWarning {nodes [
   UpdatePlayerDamageWarning() {is=title K2Node_FunctionEntry}
   > n1 = Set TempTime(TempTime:int=int + int(A=Get TempTime() -> (TempTime:int=0))) -> (Output_Get:int=0)
   > Branch(Condition="integer >= integer"(A=@n1, B=Get PlayerDamageDisplayTime() -> (PlayerDamageDisplayTime:int=0)))
   > Clear and Invalidate Timer by Handle(Handle=Get PlayerDamageTimer() -> (PlayerDamageTimer:Timer Handle Structure))
   > SetVisibility(self=Get Overlay_PlayerDamage_Warning(), InVisibility=Collapsed)
   > Set TempTime(TempTime=0) -> (Output_Get=0)
  ]}
]
variables [
 Low Energy Threshold {type=number, default=0.2}
 LowEnergyAudioComponent {type=AudioComponent, default=null}
 PlayerDamageDisplayTime {type=integer, default=5}
 PlayerDamageTimer {type=TimerHandle}
 TempTime {type=integer}
]
widgets {
 info {inheritedWidgetCount=4, parentClass=/Script/BikeKidUI.BikeKidHUD, rootWidgetClass=&Overlay, widgetCount=35}
 widgets [
  Overlay_Root {widgetClassPath=&Overlay, widget=Overlay_Root}
  Overlay_Main {
   widgetClassPath=&Overlay
   bIsVariable=true
   parent=Overlay_Root
   widget=Overlay_Main
   properties {visibility=Visible, renderTransform {angle=0, scale=<1 1>, shear=<0 0>, translation=<0 310>}}
   slot {class=OverlaySlot, properties {horizontalAlignment=HAlign_Center, verticalAlignment=VAlign_Center}}
  }
  Overlay_Reticle {
   widgetClassPath=&Overlay
   bInherited=true
   parent=Overlay_Main
   widget=Overlay_Reticle
   slot {class=OverlaySlot, properties {horizontalAlignment=HAlign_Center, verticalAlignment=VAlign_Center}}
  }
  VerticalBox_0 {
   widgetClassPath=&VerticalBox
   parent=Overlay_Reticle
   widget=VerticalBox_0
   properties {bOverride_Cursor=true}
   slot {class=OverlaySlot, properties {horizontalAlignment=HAlign_Center}}
  }
  Overlay_DistanceMeter {
   widgetClassPath=&Overlay
   parent=VerticalBox_0
   widget=Overlay_DistanceMeter
   properties {visibility=Collapsed}
   slot {class=VerticalBoxSlot, properties {horizontalAlignment=HAlign_Center, verticalAlignment=VAlign_Top, padding=<ltrb 0 0 0 10>}}
  }
  VerticalBox_2 {
   widgetClassPath=&VerticalBox
   parent=Overlay_DistanceMeter
   widget=VerticalBox_2
   slot {class=OverlaySlot, properties {horizontalAlignment=HAlign_Center, verticalAlignment=VAlign_Center}}
  }
  HorizontalBox_DistanceText {
   widgetClassPath=&HorizontalBox
   bIsVariable=true
   parent=VerticalBox_2
   widget=HorizontalBox_DistanceText
   slot {class=VerticalBoxSlot, properties {horizontalAlignment=HAlign_Left, verticalAlignment=VAlign_Top, padding=<ltrb 0 10 0 0>}}
  }
  Image_SignalIcon {
   widgetClassPath=&Image
   parent=HorizontalBox_DistanceText
   widget=Image_SignalIcon
   properties {brush {
     bIsDynamicallyLoaded=false
     drawAs=Image
     imageType=NoImage
     mirroring=NoMirror
     resourceName=None
     resourceObject=/Game/Athena/UI/InGame/Creative/PropertyEditor/Icon_Picker/Textures_V2/64x64/T_UI_IconLibrary_Wifi_64
     tiling=NoTile
     imageSize=<24 24>
     margin=<ltrb 0 0 0 0>
     outlineSettings {
      bUseBrushTransparency=false
      roundingType=HalfHeightRadius
      width=0
      color {colorUseRule=UseColor_Specified, specifiedColor=<rgba 0 0 0 0>}
      cornerRadii=<0 0 0 0>
     }
     tintColor {colorUseRule=UseColor_Specified, specifiedColor=<rgba 1 1 1 1>}
     uVRegion {bIsValid=false, max=<0 0>, min=<0 0>}
    }}
   slot {class=HorizontalBoxSlot, properties {horizontalAlignment=HAlign_Center, verticalAlignment=VAlign_Center}}
  }
  Overlay_WeakSignal_Warning {
   widgetClassPath=&Overlay
   bIsVariable=true
   parent=HorizontalBox_DistanceText
   widget=Overlay_WeakSignal_Warning
   properties {visibility=Collapsed, renderTransform {angle=0, scale=<1 1>, shear=<0 0>, translation=<8 0>}}
   slot {class=HorizontalBoxSlot}
  }
  Image_Glow_1 {
   widgetClassPath=&Image
   parent=Overlay_WeakSignal_Warning
   widget=Image_Glow_1
   properties {
    renderOpacity=0.25
    visibility=SelfHitTestInvisible
    brush {
     bIsDynamicallyLoaded=false
     drawAs=Box
     imageType=NoImage
     mirroring=NoMirror
     resourceName=None
     resourceObject=/Game/Athena/HUD/LTM/Wax/Materials/M_UI_TextGlow_Add_Gradient_Orange_Nevada
     tiling=NoTile
     imageSize=<64 64>
     margin=<ltrb 0.5 0.5 0.5 0.5>
     outlineSettings {
      bUseBrushTransparency=false
      roundingType=HalfHeightRadius
      width=0
      color {colorUseRule=UseColor_Specified, specifiedColor=<rgba 0 0 0 0>}
      cornerRadii=<0 0 0 0>
     }
     tintColor {colorUseRule=UseColor_Specified, specifiedColor=<rgba 1 1 1 1>}
     uVRegion {bIsValid=false, max=<0 0>, min=<0 0>}
    }
   }
   slot {class=OverlaySlot, properties {horizontalAlignment=HAlign_Fill, verticalAlignment=VAlign_Fill, padding=<ltrb -24 -24 -24 -24>}}
  }
  Image_OuterRing_1 {
   widgetClassPath=&Image
   parent=Overlay_WeakSignal_Warning
   widget=Image_OuterRing_1
   properties {
    renderOpacity=0.25
    brush {
     bIsDynamicallyLoaded=false
     drawAs=Box
     imageType=NoImage
     mirroring=NoMirror
     resourceName=None
     resourceObject=/Game/Athena/HUD/Vehicles/Materials/M_UI_OstrichWarningText_Outline
     tiling=NoTile
     imageSize=<16 16>
     margin=<ltrb 0.5 0.5 0.5 0.5>
     outlineSettings {
      bUseBrushTransparency=false
      roundingType=HalfHeightRadius
      width=0
      color {colorUseRule=UseColor_Specified, specifiedColor=<rgba 0 0 0 0>}
      cornerRadii=<0 0 0 0>
     }
     tintColor {colorUseRule=UseColor_Specified, specifiedColor=<rgba 1 1 1 1>}
     uVRegion {bIsValid=false, max=<0 0>, min=<0 0>}
    }
   }
   slot {class=OverlaySlot, properties {horizontalAlignment=HAlign_Fill, verticalAlignment=VAlign_Fill, padding=<ltrb -5 -5 -5 -5>}}
  }
  Image_TextBackground_1 {
   widgetClassPath=&Image
   parent=Overlay_WeakSignal_Warning
   widget=Image_TextBackground_1
   properties {
    brush {
     bIsDynamicallyLoaded=false
     drawAs=Box
     imageType=NoImage
     mirroring=NoMirror
     resourceName=None
     resourceObject=/Game/Athena/HUD/Vehicles/Materials/M_UI_OstrichWarningText_Backing
     tiling=NoTile
     imageSize=<8 8>
     margin=<ltrb 0.5 0.5 0.5 0.5>
     outlineSettings {
      bUseBrushTransparency=false
      roundingType=HalfHeightRadius
      width=0
      color {colorUseRule=UseColor_Specified, specifiedColor=<rgba 0 0 0 0>}
      cornerRadii=<0 0 0 0>
     }
     tintColor {colorUseRule=UseColor_Specified, specifiedColor=<rgba 1 1 1 1>}
     uVRegion {bIsValid=false, max=<0 0>, min=<0 0>}
    }
    renderTransformPivot=<1 0.5>
   }
   slot {class=OverlaySlot, properties {horizontalAlignment=HAlign_Fill, verticalAlignment=VAlign_Fill}}
  }
  Image_BackgroundRing_1 {
   widgetClassPath=&Image
   parent=Overlay_WeakSignal_Warning
   widget=Image_BackgroundRing_1
   properties {
    renderOpacity=0.5
    brush {
     bIsDynamicallyLoaded=false
     drawAs=Box
     imageType=NoImage
     mirroring=NoMirror
     resourceName=None
     resourceObject=&M_UI_OstrichWarningText_Outline
     tiling=NoTile
     imageSize=<8 8>
     margin=<ltrb 0.5 0.5 0.5 0.5>
     outlineSettings {
      bUseBrushTransparency=false
      roundingType=HalfHeightRadius
      width=0
      color {colorUseRule=UseColor_Specified, specifiedColor=<rgba 0 0 0 0>}
      cornerRadii=<0 0 0 0>
     }
     tintColor {colorUseRule=UseColor_Specified, specifiedColor=<rgba 1 1 1 1>}
     uVRegion {bIsValid=false, max=<0 0>, min=<0 0>}
    }
   }
   slot {class=OverlaySlot, properties {horizontalAlignment=HAlign_Fill, verticalAlignment=VAlign_Fill}}
  }
  Text_WeakSignal {
   widgetClassPath=&CommonTextBlock
   parent=Overlay_WeakSignal_Warning
   widget=Text_WeakSignal
   properties {
    justification=Center
    mobileFontSizeMultiplier=1.4
    text="Weak Signal!"
    textTransformPolicy=ToUpper
    colorAndOpacity {colorUseRule=UseColor_Specified, specifiedColor=<rgba 0 0 0 1>}
    font {
     bForceMonospaced=false
     bMaterialIsStencil=false
     fontMaterial=null
     fontObject=&BurbankSmall
     letterSpacing=100
     monospacedWidth=1
     size=9
     skewAmount=0.25
     typefaceFontName=Black
     outlineSettings {
      bApplyOutlineToDropShadows=false
      bMiteredCorners=false
      bSeparateFillAlpha=true
      outlineMaterial=null
      outlineSize=0
      outlineColor=<rgba 1 1 1 0.125>
     }
    }
    shadowOffset=<0 0>
   }
   slot {class=OverlaySlot, properties {horizontalAlignment=HAlign_Fill, verticalAlignment=VAlign_Center, padding=<ltrb 12 2 12 2>}}
  }
  Overlay_Health {
   widgetClassPath=&Overlay
   bInherited=true
   parent=VerticalBox_0
   widget=Overlay_Health
   slot {class=VerticalBoxSlot, properties {verticalAlignment=VAlign_Top}}
  }
  GridPanel_0 {widgetClassPath=&GridPanel, parent=Overlay_Health, widget=GridPanel_0, slot {class=OverlaySlot}}
  Image_HealthBar {
   widgetClassPath=&Image
   bIsVariable=true
   parent=GridPanel_0
   widget=Image_HealthBar
   properties {brush {
     bIsDynamicallyLoaded=false
     drawAs=Image
     imageType=NoImage
     mirroring=NoMirror
     resourceName=None
     resourceObject=/BikeKid/UI/Materials/M_UI_BikeKid_Health_WithBacking
     tiling=NoTile
     imageSize=<320 15>
     margin=<ltrb 0 0 0 0>
     outlineSettings {
      bUseBrushTransparency=false
      roundingType=HalfHeightRadius
      width=0
      color {colorUseRule=UseColor_Specified, specifiedColor=<rgba 0 0 0 0>}
      cornerRadii=<0 0 0 0>
     }
     tintColor {colorUseRule=UseColor_Specified, specifiedColor=<rgba 1 1 1 1>}
     uVRegion {bIsValid=false, max=<0 0>, min=<0 0>}
    }}
   slot {class=GridSlot, properties {horizontalAlignment=HAlign_Left, layer=1, verticalAlignment=VAlign_Top}}
  }
  HorizontalBox_HealthText {
   widgetClassPath=&HorizontalBox
   bIsVariable=true
   parent=GridPanel_0
   widget=HorizontalBox_HealthText
   properties {visibility=Collapsed, renderTransform {angle=0, scale=<1 1>, shear=<0 0>, translation=<0 -8>}}
   slot {class=GridSlot, properties {horizontalAlignment=HAlign_Left, layer=1, verticalAlignment=VAlign_Top, padding=<ltrb 0 -20 0 0>}}
  }
  Image_HealthIcon {
   widgetClassPath=&Image
   bIsVariable=true
   parent=HorizontalBox_HealthText
   widget=Image_HealthIcon
   properties {brush {
     bIsDynamicallyLoaded=false
     drawAs=Image
     imageType=NoImage
     mirroring=NoMirror
     resourceName=None
     resourceObject=/Game/Athena/HUD/Art/icon_health
     tiling=NoTile
     imageSize=<16 16>
     margin=<ltrb 0 0 0 0>
     outlineSettings {
      bUseBrushTransparency=false
      roundingType=HalfHeightRadius
      width=0
      color {colorUseRule=UseColor_Specified, specifiedColor=<rgba 0 0 0 0>}
      cornerRadii=<0 0 0 0>
     }
     tintColor {colorUseRule=UseColor_Specified, specifiedColor=<rgba 1 1 1 1>}
     uVRegion {bIsValid=false, max=<0 0>, min=<0 0>}
    }}
   slot {class=HorizontalBoxSlot, properties {horizontalAlignment=HAlign_Right, verticalAlignment=VAlign_Center, padding=<ltrb 0 0 5 0>}}
  }
  Text_Health {
   widgetClassPath=&CommonTextBlock
   bIsVariable=true
   parent=HorizontalBox_HealthText
   widget=Text_Health
   properties {
    mobileFontSizeMultiplier=1.6
    text=100
    font {
     bForceMonospaced=false
     bMaterialIsStencil=false
     fontMaterial=null
     fontObject=&BurbankSmall
     letterSpacing=0
     monospacedWidth=1
     size=12
     skewAmount=0
     typefaceFontName=Black
     outlineSettings {
      bApplyOutlineToDropShadows=false
      bMiteredCorners=false
      bSeparateFillAlpha=false
      outlineMaterial=null
      outlineSize=0
      outlineColor=<rgba 0 0 0 0.5>
     }
    }
    shadowColorAndOpacity=<rgba 0 0 0 0.5>
    shadowOffset=<0 0>
   }
   slot {class=HorizontalBoxSlot, properties {horizontalAlignment=HAlign_Center, verticalAlignment=VAlign_Center}}
  }
  Overlay_LowHealth_Warning {
   widgetClassPath=&Overlay
   bIsVariable=true
   parent=VerticalBox_0
   widget=Overlay_LowHealth_Warning
   properties {visibility=Collapsed}
   slot {class=VerticalBoxSlot, properties {padding=<ltrb 0 12 0 0>}}
  }
  Image_Glow {
   widgetClassPath=&Image
   parent=Overlay_LowHealth_Warning
   widget=Image_Glow
   properties {
    renderOpacity=0.25
    visibility=SelfHitTestInvisible
    brush {
     bIsDynamicallyLoaded=false
     drawAs=Box
     imageType=NoImage
     mirroring=NoMirror
     resourceName=None
     resourceObject=&M_UI_TextGlow_Add_Gradient_Orange_Nevada
     tiling=NoTile
     imageSize=<64 64>
     margin=<ltrb 0.5 0.5 0.5 0.5>
     outlineSettings {
      bUseBrushTransparency=false
      roundingType=HalfHeightRadius
      width=0
      color {colorUseRule=UseColor_Specified, specifiedColor=<rgba 0 0 0 0>}
      cornerRadii=<0 0 0 0>
     }
     tintColor {colorUseRule=UseColor_Specified, specifiedColor=<rgba 1 1 1 1>}
     uVRegion {bIsValid=false, max=<0 0>, min=<0 0>}
    }
   }
   slot {class=OverlaySlot, properties {horizontalAlignment=HAlign_Fill, verticalAlignment=VAlign_Fill, padding=<ltrb -24 -24 -24 -24>}}
  }
  Image_OuterRing {
   widgetClassPath=&Image
   parent=Overlay_LowHealth_Warning
   widget=Image_OuterRing
   properties {
    renderOpacity=0.25
    brush {
     bIsDynamicallyLoaded=false
     drawAs=Box
     imageType=NoImage
     mirroring=NoMirror
     resourceName=None
     resourceObject=&M_UI_OstrichWarningText_Outline
     tiling=NoTile
     imageSize=<16 16>
     margin=<ltrb 0.5 0.5 0.5 0.5>
     outlineSettings {
      bUseBrushTransparency=false
      roundingType=HalfHeightRadius
      width=0
      color {colorUseRule=UseColor_Specified, specifiedColor=<rgba 0 0 0 0>}
      cornerRadii=<0 0 0 0>
     }
     tintColor {colorUseRule=UseColor_Specified, specifiedColor=<rgba 1 1 1 1>}
     uVRegion {bIsValid=false, max=<0 0>, min=<0 0>}
    }
   }
   slot {class=OverlaySlot, properties {horizontalAlignment=HAlign_Fill, verticalAlignment=VAlign_Fill, padding=<ltrb -5 -5 -5 -5>}}
  }
  Image_TextBackground {
   widgetClassPath=&Image
   parent=Overlay_LowHealth_Warning
   widget=Image_TextBackground
   properties {
    brush {
     bIsDynamicallyLoaded=false
     drawAs=Box
     imageType=NoImage
     mirroring=NoMirror
     resourceName=None
     resourceObject=&M_UI_OstrichWarningText_Backing
     tiling=NoTile
     imageSize=<8 25>
     margin=<ltrb 0.5 0.5 0.5 0.5>
     outlineSettings {
      bUseBrushTransparency=false
      roundingType=HalfHeightRadius
      width=0
      color {colorUseRule=UseColor_Specified, specifiedColor=<rgba 0 0 0 0>}
      cornerRadii=<0 0 0 0>
     }
     tintColor {colorUseRule=UseColor_Specified, specifiedColor=<rgba 1 1 1 1>}
     uVRegion {bIsValid=false, max=<0 0>, min=<0 0>}
    }
    renderTransformPivot=<1 0.5>
   }
   slot {class=OverlaySlot, properties {horizontalAlignment=HAlign_Fill, verticalAlignment=VAlign_Fill}}
  }
  Image_BackgroundRing {
   widgetClassPath=&Image
   parent=Overlay_LowHealth_Warning
   widget=Image_BackgroundRing
   properties {
    renderOpacity=0.5
    brush {
     bIsDynamicallyLoaded=false
     drawAs=Box
     imageType=NoImage
     mirroring=NoMirror
     resourceName=None
     resourceObject=&M_UI_OstrichWarningText_Outline
     tiling=NoTile
     imageSize=<8 8>
     margin=<ltrb 0.5 0.5 0.5 0.5>
     outlineSettings {
      bUseBrushTransparency=false
      roundingType=HalfHeightRadius
      width=0
      color {colorUseRule=UseColor_Specified, specifiedColor=<rgba 0 0 0 0>}
      cornerRadii=<0 0 0 0>
     }
     tintColor {colorUseRule=UseColor_Specified, specifiedColor=<rgba 1 1 1 1>}
     uVRegion {bIsValid=false, max=<0 0>, min=<0 0>}
    }
   }
   slot {class=OverlaySlot, properties {horizontalAlignment=HAlign_Fill, verticalAlignment=VAlign_Fill}}
  }
  Text_LowHealth {
   widgetClassPath=&CommonTextBlock
   parent=Overlay_LowHealth_Warning
   widget=Text_LowHealth
   properties {
    justification=Center
    mobileFontSizeMultiplier=1.4
    text="Low Health!"
    textTransformPolicy=ToUpper
    colorAndOpacity {colorUseRule=UseColor_Specified, specifiedColor=<rgba 0 0 0 1>}
    font {
     bForceMonospaced=false
     bMaterialIsStencil=false
     fontMaterial=null
     fontObject=&BurbankSmall
     letterSpacing=100
     monospacedWidth=1
     size=9
     skewAmount=0.25
     typefaceFontName=Black
     outlineSettings {
      bApplyOutlineToDropShadows=false
      bMiteredCorners=false
      bSeparateFillAlpha=true
      outlineMaterial=null
      outlineSize=0
      outlineColor=<rgba 1 1 1 0.125>
     }
    }
    shadowOffset=<0 0>
   }
   slot {class=OverlaySlot, properties {horizontalAlignment=HAlign_Fill, verticalAlignment=VAlign_Center, padding=<ltrb 12 2 12 2>}}
  }
  Overlay_PlayerDamage_Warning {
   widgetClassPath=&Overlay
   bIsVariable=true
   parent=Overlay_Reticle
   widget=Overlay_PlayerDamage_Warning
   properties {visibility=Collapsed, renderTransform {angle=0, scale=<1 1>, shear=<0 0>, translation=<3 -8>}}
   slot {class=OverlaySlot, properties {horizontalAlignment=HAlign_Center, verticalAlignment=VAlign_Bottom, padding=<ltrb 0 0 0 -50>}}
  }
  Image_damagearrow {
   widgetClassPath=&Image
   bIsVariable=true
   parent=Overlay_PlayerDamage_Warning
   widget=Image_damagearrow
   properties {
    brush {
     bIsDynamicallyLoaded=false
     drawAs=Image
     imageType=NoImage
     mirroring=NoMirror
     resourceName=None
     resourceObject=/Game/Athena/HUD/LTM/Mash/Materials/MI_UI_WarningArrow
     tiling=NoTile
     imageSize=<32 32>
     margin=<ltrb 0 0 0 0>
     outlineSettings {
      bUseBrushTransparency=false
      roundingType=HalfHeightRadius
      width=0
      color {colorUseRule=UseColor_Specified, specifiedColor=<rgba 0 0 0 0>}
      cornerRadii=<0 0 0 0>
     }
     tintColor {colorUseRule=UseColor_Specified, specifiedColor=<rgba 1 1 1 1>}
     uVRegion {bIsValid=false, max=<0 0>, min=<0 0>}
    }
    renderTransform {angle=45, scale=<2 2>, shear=<0 0>, translation=<-101 11>}
   }
   slot {class=OverlaySlot, properties {horizontalAlignment=HAlign_Center, verticalAlignment=VAlign_Bottom}}
  }
  Image_Glow_2 {
   widgetClassPath=&Image
   parent=Overlay_PlayerDamage_Warning
   widget=Image_Glow_2
   properties {
    renderOpacity=0.25
    visibility=SelfHitTestInvisible
    brush {
     bIsDynamicallyLoaded=false
     drawAs=Box
     imageType=NoImage
     mirroring=NoMirror
     resourceName=None
     resourceObject=&M_UI_TextGlow_Add_Gradient_Orange_Nevada
     tiling=NoTile
     imageSize=<64 64>
     margin=<ltrb 0.5 0.5 0.5 0.5>
     outlineSettings {
      bUseBrushTransparency=false
      roundingType=HalfHeightRadius
      width=0
      color {colorUseRule=UseColor_Specified, specifiedColor=<rgba 0 0 0 0>}
      cornerRadii=<0 0 0 0>
     }
     tintColor {colorUseRule=UseColor_Specified, specifiedColor=<rgba 1 1 1 1>}
     uVRegion {bIsValid=false, max=<0 0>, min=<0 0>}
    }
   }
   slot {class=OverlaySlot, properties {horizontalAlignment=HAlign_Fill, verticalAlignment=VAlign_Fill, padding=<ltrb -24 -24 -24 -24>}}
  }
  Image_OuterRing_2 {
   widgetClassPath=&Image
   parent=Overlay_PlayerDamage_Warning
   widget=Image_OuterRing_2
   properties {
    renderOpacity=0.25
    brush {
     bIsDynamicallyLoaded=false
     drawAs=Box
     imageType=NoImage
     mirroring=NoMirror
     resourceName=None
     resourceObject=&M_UI_OstrichWarningText_Outline
     tiling=NoTile
     imageSize=<16 16>
     margin=<ltrb 0.5 0.5 0.5 0.5>
     outlineSettings {
      bUseBrushTransparency=false
      roundingType=HalfHeightRadius
      width=0
      color {colorUseRule=UseColor_Specified, specifiedColor=<rgba 0 0 0 0>}
      cornerRadii=<0 0 0 0>
     }
     tintColor {colorUseRule=UseColor_Specified, specifiedColor=<rgba 1 1 1 1>}
     uVRegion {bIsValid=false, max=<0 0>, min=<0 0>}
    }
   }
   slot {class=OverlaySlot, properties {horizontalAlignment=HAlign_Fill, verticalAlignment=VAlign_Fill, padding=<ltrb -5 -5 -5 -5>}}
  }
  Image_TextBackground_2 {
   widgetClassPath=&Image
   parent=Overlay_PlayerDamage_Warning
   widget=Image_TextBackground_2
   properties {
    brush {
     bIsDynamicallyLoaded=false
     drawAs=Box
     imageType=NoImage
     mirroring=NoMirror
     resourceName=None
     resourceObject=&M_UI_OstrichWarningText_Backing
     tiling=NoTile
     imageSize=<8 8>
     margin=<ltrb 0.5 0.5 0.5 0.5>
     outlineSettings {
      bUseBrushTransparency=false
      roundingType=HalfHeightRadius
      width=0
      color {colorUseRule=UseColor_Specified, specifiedColor=<rgba 0 0 0 0>}
      cornerRadii=<0 0 0 0>
     }
     tintColor {colorUseRule=UseColor_Specified, specifiedColor=<rgba 1 1 1 1>}
     uVRegion {bIsValid=false, max=<0 0>, min=<0 0>}
    }
    renderTransformPivot=<1 0.5>
   }
   slot {class=OverlaySlot, properties {horizontalAlignment=HAlign_Fill, verticalAlignment=VAlign_Fill}}
  }
  Image_BackgroundRing_2 {
   widgetClassPath=&Image
   parent=Overlay_PlayerDamage_Warning
   widget=Image_BackgroundRing_2
   properties {
    renderOpacity=0.5
    brush {
     bIsDynamicallyLoaded=false
     drawAs=Box
     imageType=NoImage
     mirroring=NoMirror
     resourceName=None
     resourceObject=&M_UI_OstrichWarningText_Outline
     tiling=NoTile
     imageSize=<8 8>
     margin=<ltrb 0.5 0.5 0.5 0.5>
     outlineSettings {
      bUseBrushTransparency=false
      roundingType=HalfHeightRadius
      width=0
      color {colorUseRule=UseColor_Specified, specifiedColor=<rgba 0 0 0 0>}
      cornerRadii=<0 0 0 0>
     }
     tintColor {colorUseRule=UseColor_Specified, specifiedColor=<rgba 1 1 1 1>}
     uVRegion {bIsValid=false, max=<0 0>, min=<0 0>}
    }
   }
   slot {class=OverlaySlot, properties {horizontalAlignment=HAlign_Fill, verticalAlignment=VAlign_Fill}}
  }
  Text_PlayerDamage {
   widgetClassPath=&CommonTextBlock
   parent=Overlay_PlayerDamage_Warning
   widget=Text_PlayerDamage
   properties {
    justification=Center
    mobileFontSizeMultiplier=1.4
    text="Operator taking damage!"
    textTransformPolicy=ToUpper
    colorAndOpacity {colorUseRule=UseColor_Specified, specifiedColor=<rgba 0 0 0 1>}
    font {
     bForceMonospaced=false
     bMaterialIsStencil=false
     fontMaterial=null
     fontObject=&BurbankSmall
     letterSpacing=100
     monospacedWidth=1
     size=9
     skewAmount=0.25
     typefaceFontName=Black
     outlineSettings {
      bApplyOutlineToDropShadows=false
      bMiteredCorners=false
      bSeparateFillAlpha=true
      outlineMaterial=null
      outlineSize=0
      outlineColor=<rgba 1 1 1 0.125>
     }
    }
    shadowOffset=<0 0>
   }
   slot {class=OverlaySlot, properties {horizontalAlignment=HAlign_Center, verticalAlignment=VAlign_Center, padding=<ltrb 12 2 12 2>}}
  }
  Overlay_TetheringIndicator {
   widgetClassPath=&Overlay
   bInherited=true
   parent=Overlay_Root
   widget=Overlay_TetheringIndicator
   properties {visibility=Collapsed}
   slot {class=OverlaySlot, properties {horizontalAlignment=HAlign_Fill, verticalAlignment=VAlign_Fill}}
  }
  Overlay_EnergyLevel {widgetClassPath=&Overlay, bInherited=true, parent=Overlay_Root, widget=Overlay_EnergyLevel, slot {class=OverlaySlot}}
 ]
}
