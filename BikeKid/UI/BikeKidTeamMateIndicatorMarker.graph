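# World marker widget for BikeKid that shows which non-local player owns a marked bike, built on AthenaMarkedActorIndicator.
# Its layout pairs a team indicator arrow image with a RidingMarkerInfo player icon widget.
# OnSetMarkerData casts the marked actor to BP_BikeKid, resolves its OwningPawn's player state and initializes the icon from it.
# RidingMarkerInfo collapses when the owner is locally controlled or any validity or cast check fails; the arrow is never toggled here.
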WidgetBlueprint /BikeKid/UI/BikeKidTeamMateIndicatorMarker
graphs [EventGraph {nodes [
   n1 = Event PreConstruct()
   > Parent: PreConstruct(IsDesignTime=@n1.IsDesignTime)
   Event Construct()
   > Parent: Construct()
   n2 = Event Tick()
   > Parent: Tick!(execute, MyGeometry:Geometry Structure=@n2.MyGeometry, InDeltaTime:float=@n2.InDeltaTime) -> (then) {is=title K2Node_CallParentFunction}
   n3 = Event OnSetMarkerData!() -> (OutputDelegate, then, FortWorldMarkerData:"Fort World Marker Data Structure (by ref)") {is=title K2Node_Event}
   > Parent: OnSetMarkerData!(execute, FortWorldMarkerData:"Fort World Marker Data Structure (by ref)"=@n3.FortWorldMarkerData) -> (then) {is=title K2Node_CallParentFunction}
   > n4 = Cast To BP_BikeKid!(
    execute
    Object=Break Fort World Marker Data!(FortWorldMarkerData:"Fort World Marker Data Structure (by ref)"=@n3.FortWorldMarkerData) -> (
     MarkerID:Marker ID Structure
     Owner:Fort Player State Weak object
     MarkerType:EFortWorldMarkerType Enum
     BasePositionOffset:Vector=~
     WorldNormal:Vector=~
     ItemDefinition:Item Definition Base Object Reference
     ItemCount:int
     PinnedQuestIndex:int
     MarkedActor:Actor Weak object
     MarkedActorClass:Actor Soft Class Reference
     MarkerTag:Gameplay Tag Structure
     bUseHoveredMarkerDetail:bool=~
    ) {is=title K2Node_BreakStruct}.MarkedActor
   ) -> (then, CastFailed, AsBP Bike Kid:BP Bike Kid Object Reference) {is=title K2Node_DynamicCast}
   > n5 = Is Valid(InputObject=@n4)
   n6 = Get OwningPawn!(self:BP Bike Kid Object Reference=@n4) -> (OwningPawn:Fort Player Pawn Object Reference)
   > n7 = Is Valid(InputObject=@n6)
   > n8 = Branch(Condition=NOT Boolean(A=IsLocallyControlled!(self:Pawn Object Reference=@n6) -> (ReturnValue:bool)))
   > n9 = AttemptGetPlayerStateForPawn!(execute, FortPawn:Fort Player Pawn Object Reference=@n6) -> (then, ReturnValue:Fort Player State Object Reference)
   > n10 = Is Valid(InputObject=@n9)
   > n11 = Cast To FortPlayerStateAthena!(execute, Object=@n9) -> (then, CastFailed, AsFort Player State Athena:Fort Player State Athena Object Reference) {is=title K2Node_DynamicCast}
   > InitPlayerIconState!(
    execute
    self:Converted Marker Info Object Reference=Get RidingMarkerInfo() -> (RidingMarkerInfo:Converted Marker Info Object Reference)
    PSA:Fort Player State Athena Object Reference=@n11
   ) -> (then)
   > SetVisibility(self=Get RidingMarkerInfo(), InVisibility=HitTestInvisible)
   SetVisibility(
    execute=[@n11.CastFailed, @n10.Is Not Valid, @n8.else, @n7.Is Not Valid, @n4.CastFailed, @n5.Is Not Valid]
    self=Get RidingMarkerInfo()
    InVisibility=Collapsed
   )
  ]}]
variables []
widgets {
 info {inheritedWidgetCount=5, parentClass=/Script/FortniteUI.AthenaMarkedActorIndicator, rootWidgetClass=&Overlay, widgetCount=9}
 widgets [
  Overlay_19 {widgetClassPath=&Overlay, widget=Overlay_19}
  SizeBox_52 {
   widgetClassPath=&SizeBox
   parent=Overlay_19
   widget=SizeBox_52
   slot {class=OverlaySlot, properties {horizontalAlignment=HAlign_Center, verticalAlignment=VAlign_Fill}}
  }
  IrwinIndicator {
   widgetClassPath=&Image
   bIsVariable=true
   parent=SizeBox_52
   widget=IrwinIndicator
   properties {brush {
     bIsDynamicallyLoaded=false
     drawAs=Image
     imageType=FullColor
     mirroring=NoMirror
     resourceName=None
     resourceObject=/Game/Athena/HUD/Team/MeshIndicators/T_TeamIndicatorArrow
     tiling=NoTile
     imageSize=<64 64>
     margin=<ltrb 0 0 0 0>
     outlineSettings {
      bUseBrushTransparency=false
      roundingType=HalfHeightRadius
      width=0
      color {colorUseRule=UseColor_Specified, specifiedColor=<rgba 0 0 0 0>}
      cornerRadii=<0 0 0 0>
     }
     tintColor {colorUseRule=UseColor_Specified, specifiedColor=<rgba 1 1 1 1>}
     uVRegion {bIsValid=false, max=<0 0>, min=<0 0>}
    }}
   slot {class=SizeBoxSlot, properties {padding=<ltrb 0 0 0 10>}}
  }
  RidingMarkerInfo {
   widgetClassPath=/Game/Athena/HUD/Actors/ConvertedMarkerInfo
   bIsVariable=true
   parent=Overlay_19
   widget=RidingMarkerInfo
   slot {class=OverlaySlot, properties {horizontalAlignment=HAlign_Center, verticalAlignment=VAlign_Fill, padding=<ltrb 0 -10 0 50>}}
  }
  Switcher_IndicatorSize {widgetClassPath=&WidgetSwitcher, bInherited=true, widget=null}
  Panel_SmallIndicator {widgetClassPath=/Script/UMG.PanelWidget, bInherited=true, widget=null}
  Panel_LargeIndicator {widgetClassPath=&PanelWidget, bInherited=true, widget=null}
  Text_Distance {widgetClassPath=&CommonTextBlock, bInherited=true, widget=null}
  TutorialHighlight {widgetClassPath=&TutorialHighlight}
 ]
}
