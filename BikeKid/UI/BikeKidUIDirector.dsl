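; BikeKidUIDirector is a UI director Blueprint that shows and hides the BikeKid UI.
; It listens for the BikeKid deployed event and adds BikeKidUIScene through the DynamicUIDirector.
; On the BikeKid dismissed event it removes the same scene, while BeginPlay and the construction script stay empty.
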(event EventBeginPlay)

(event BikeKidEvents|EventOnBikeKidDeployed
  (DynamicUIDirector|AddScene "/BikeKid/UI/BikeKidUIScene.BikeKidUIScene"))

(event BikeKidEvents|EventOnBikeKidDismissed
  (DynamicUIDirector|RemoveScene "/BikeKid/UI/BikeKidUIScene.BikeKidUIScene"))

(fn ConstructionScript)
